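# Sound cue for an outdoor ambient one-shot spot sound of earth movement, with rubble, rock and dirt swells, in the LostBase audio set.
# A random node picks among six equally weighted wave variations and feeds a modulator that adds slight pitch and volume variation.
# It uses the tile ambient one-shot attenuation and the outdoor ambient one-shots sound class.
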SoundCue /Hermes/Audio/LostBase/Spots/SC_Spot_Earth_Movement_Rubble_Rock_Dirt_Swell
root=Modulator_0
nodes [
 {id=Modulator_0, children [Random_0], properties {pitchMax=1.1, pitchMin=0.9, volumeMax=1.1}}
 {id=Random_0, children [WavePlayer_0,WavePlayer_1,WavePlayer_5,WavePlayer_6,WavePlayer_7,WavePlayer_8], properties {weights [1,1,1,1,1,1]}}
 {id=WavePlayer_0, properties {soundWaveAssetPtr=&/Spot_Earth_Movement_Rubble_Rock_Dirt_Swell_01}}
 {id=WavePlayer_1, properties {soundWaveAssetPtr=&/Spot_Earth_Movement_Rubble_Rock_Dirt_Swell_02}}
 {id=WavePlayer_5, properties {soundWaveAssetPtr=&/Spot_Earth_Movement_Rubble_Rock_Dirt_Swell_03}}
 {id=WavePlayer_6, properties {soundWaveAssetPtr=&/Spot_Earth_Movement_Rubble_Rock_Dirt_Swell_04}}
 {id=WavePlayer_7, properties {soundWaveAssetPtr=&/Spot_Earth_Movement_Rubble_Rock_Dirt_Swell_05}}
 {id=WavePlayer_8, properties {soundWaveAssetPtr=&/Spot_Earth_Movement_Rubble_Rock_Dirt_Swell_06}}
]
properties {
 attenuationSettings=/Game/Sounds/Attenuation/Fort_TileAmbient_OneShot_Att
 maxDistance=8000
 soundClassObject=&Fort_Ambient_OneShots_Outdoors
 volumeMultiplier=1
}
