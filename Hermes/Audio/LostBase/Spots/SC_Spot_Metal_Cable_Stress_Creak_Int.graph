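# Sound cue for an interior metal cable stress creak one-shot, used as an ambient spot sound in the Lost Base area.
# A Random node picks one of nine equally weighted creak waves, then a Modulator varies pitch slightly downward and randomizes volume.
# It plays at reduced volume in the indoor ambient one-shots sound class, using a borrowed distant-creature one-shot attenuation for spatial falloff.
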SoundCue /Hermes/Audio/LostBase/Spots/SC_Spot_Metal_Cable_Stress_Creak_Int
root=Modulator_0
nodes [
 {id=Modulator_0, children [Random_1], properties {pitchMax=1, pitchMin=0.8, volumeMax=1.2}}
 {
  id=Random_1
  children [WavePlayer_2, WavePlayer_3, WavePlayer_4, WavePlayer_5, WavePlayer_6, WavePlayer_7, WavePlayer_8, WavePlayer_9, WavePlayer_10]
  properties {weights [1,1,1,1,1,1,1,1,1]}
 }
 {id=WavePlayer_10, properties {soundWaveAssetPtr=&/Spot_Metal_Cable_Stress_Creak_Int_09}}
 {id=WavePlayer_2, properties {soundWaveAssetPtr=&/Spot_Metal_Cable_Stress_Creak_Int_01}}
 {id=WavePlayer_3, properties {soundWaveAssetPtr=&/Spot_Metal_Cable_Stress_Creak_Int_02}}
 {id=WavePlayer_4, properties {soundWaveAssetPtr=&/Spot_Metal_Cable_Stress_Creak_Int_03}}
 {id=WavePlayer_5, properties {soundWaveAssetPtr=&/Spot_Metal_Cable_Stress_Creak_Int_04}}
 {id=WavePlayer_6, properties {soundWaveAssetPtr=&/Spot_Metal_Cable_Stress_Creak_Int_05}}
 {id=WavePlayer_7, properties {soundWaveAssetPtr=&/Spot_Metal_Cable_Stress_Creak_Int_06}}
 {id=WavePlayer_8, properties {soundWaveAssetPtr=&/Spot_Metal_Cable_Stress_Creak_Int_07}}
 {id=WavePlayer_9, properties {soundWaveAssetPtr=&/Spot_Metal_Cable_Stress_Creak_Int_08}}
]
properties {
 attenuationSettings=/AmbientAudioBase/Attenuations/Fort_OneShot_Distant_Creature_Att
 maxDistance=8500
 soundClassObject=/Game/Sounds/FortSoundClassesAndModes/Fort_Ambient_OneShots_Indoors
 volumeMultiplier=0.7
}
