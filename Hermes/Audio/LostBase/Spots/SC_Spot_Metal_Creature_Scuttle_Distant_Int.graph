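# Sound cue for a distant one-shot spot ambience of a creature scuttling on metal, set up for LostBase interiors.
# A Random node picks one of three equally weighted scuttle recordings, and a Modulator above it varies pitch slightly downward and volume slightly upward on each play.
# It uses a distant-creature attenuation asset and the indoor ambient one-shot sound class, with the overall cue volume reduced.
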SoundCue /Hermes/Audio/LostBase/Spots/SC_Spot_Metal_Creature_Scuttle_Distant_Int
root=Modulator_0
nodes [
 {id=Modulator_0, children [Random_1], properties {pitchMax=1, pitchMin=0.8, volumeMax=1.2, volumeMin=1}}
 {id=Random_1, children [WavePlayer_2,WavePlayer_3,WavePlayer_4], properties {weights [1,1,1]}}
 {id=WavePlayer_2, properties {soundWaveAssetPtr=&/Spot_Metal_Creature_Scuttle_Distant_01}}
 {id=WavePlayer_3, properties {soundWaveAssetPtr=&/Spot_Metal_Creature_Scuttle_Distant_02}}
 {id=WavePlayer_4, properties {soundWaveAssetPtr=&/Spot_Metal_Creature_Scuttle_Distant_03}}
]
properties {
 attenuationSettings=/AmbientAudioBase/Attenuations/Fort_OneShot_Distant_Creature_Att
 maxDistance=8500
 soundClassObject=/Game/Sounds/FortSoundClassesAndModes/Fort_Ambient_OneShots_Indoors
 volumeMultiplier=0.6
}
