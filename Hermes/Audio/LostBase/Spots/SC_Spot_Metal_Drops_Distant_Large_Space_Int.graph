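# Sound cue for ambient one-shot spot sounds of distant metal drops in a large interior space, part of the Hermes Lost Base audio.
# A Random node picks one of eight equally weighted wave variations.
# A root Modulator varies pitch, at or below the original, and volume on each play to mask repetition.
# It plays at reduced volume through the Fort ambient one-shots sound class, using a shared distant creature attenuation asset.
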SoundCue /Hermes/Audio/LostBase/Spots/SC_Spot_Metal_Drops_Distant_Large_Space_Int
root=Modulator_0
nodes [
 {id=Modulator_0, children [Random_0], properties {pitchMax=1, pitchMin=0.8, volumeMax=1.2}}
 {
  id=Random_0
  children [WavePlayer_0, WavePlayer_1, WavePlayer_13, WavePlayer_14, WavePlayer_15, WavePlayer_16, WavePlayer_17, WavePlayer_18]
  properties {weights [1,1,1,1,1,1,1,1]}
 }
 {id=WavePlayer_0, properties {soundWaveAssetPtr=&/Spot_Metal_Drops_Distant_Large_Space_Int_01}}
 {id=WavePlayer_1, properties {soundWaveAssetPtr=&/Spot_Metal_Drops_Distant_Large_Space_Int_02}}
 {id=WavePlayer_13, properties {soundWaveAssetPtr=&/Spot_Metal_Drops_Distant_Large_Space_Int_03}}
 {id=WavePlayer_14, properties {soundWaveAssetPtr=&/Spot_Metal_Drops_Distant_Large_Space_Int_04}}
 {id=WavePlayer_15, properties {soundWaveAssetPtr=&/Spot_Metal_Drops_Distant_Large_Space_Int_05}}
 {id=WavePlayer_16, properties {soundWaveAssetPtr=&/Spot_Metal_Drops_Distant_Large_Space_Int_06}}
 {id=WavePlayer_17, properties {soundWaveAssetPtr=&/Spot_Metal_Drops_Distant_Large_Space_Int_07}}
 {id=WavePlayer_18, properties {soundWaveAssetPtr=&/Spot_Metal_Drops_Distant_Large_Space_Int_08}}
]
properties {
 attenuationSettings=/AmbientAudioBase/Attenuations/Fort_OneShot_Distant_Creature_Att
 maxDistance=8500
 soundClassObject=&Fort_Ambient_OneShots
 volumeMultiplier=0.5
}
