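# Looping spot ambience cue for a geothermal vent's steam and smoke in LostBase.
# It mixes a louder main loop, randomly picked from three waves with slight pitch variation, with a quieter low loop randomly picked from three others.
# A quality node feeds this mix to the first two levels and leaves the other two empty.
# Overrides set capsule logarithmic attenuation with listener focus and a small concurrency cap.
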SoundCue /Hermes/Audio/LostBase/Spots/Spot_Geothermal_Vent_Steam_Smoke_Loop_Cue
root=QualityLevel_0
nodes [
 {id=Mixer_1, children [Modulator_0,Random_0], properties {inputVolume [1.5,0.5]}}
 {id=Modulator_0, children [Random_1], properties {pitchMax=1.1, pitchMin=1}}
 {id=QualityLevel_0, children [Mixer_1,Mixer_1,null,null]}
 {id=Random_0, children [WavePlayer_4,WavePlayer_5,WavePlayer_6], properties {weights [1,1,1]}}
 {id=Random_1, children [WavePlayer_7,WavePlayer_8,WavePlayer_9], properties {weights [1,1,1]}}
 {id=WavePlayer_4, properties {bLooping=true, soundWaveAssetPtr=&/Spot_Geothermal_Vent_Steam_Smoke_Low_Loop_01}}
 {id=WavePlayer_5, properties {bLooping=true, soundWaveAssetPtr=&/Spot_Geothermal_Vent_Steam_Smoke_Low_Loop_02}}
 {id=WavePlayer_6, properties {bLooping=true, soundWaveAssetPtr=&/Spot_Geothermal_Vent_Steam_Smoke_Low_Loop_03}}
 {id=WavePlayer_7, properties {bLooping=true, soundWaveAssetPtr=&/Spot_Geothermal_Vent_Steam_Smoke_Loop_01}}
 {id=WavePlayer_8, properties {bLooping=true, soundWaveAssetPtr=&/Spot_Geothermal_Vent_Steam_Smoke_Loop_02}}
 {id=WavePlayer_9, properties {bLooping=true, soundWaveAssetPtr=&/Spot_Geothermal_Vent_Steam_Smoke_Loop_03}}
]
properties {
 bOverrideAttenuation=true
 bOverrideConcurrency=true
 maxDistance=1700
 attenuationOverrides {
  attenuationShape=Capsule
  bEnableFocusInterpolation=true
  bEnableListenerFocus=true
  distanceAlgorithm=Logarithmic
  falloffDistance=850
  focusAttackInterpSpeed=0.05
  focusReleaseInterpSpeed=0.01
  nonFocusVolumeAttenuation=0.5
  attenuationShapeExtents=<850 300>
 }
 concurrencyOverrides {maxCount=3}
}
